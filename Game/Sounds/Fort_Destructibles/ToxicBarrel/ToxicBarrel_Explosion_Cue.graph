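# Sound cue for the toxic barrel explosion, layering close blast, wet goop, and medium and distant tails.
# Each layer picks a random wave with pitch or volume modulation and has its own distance attenuation, with separate sound classes for close and distant groups.
# A quality node gives its first two slots the full mix, including a dual-attenuated liquid decal sweetener.
# The last two get a mix without that sweetener and with fewer variations.
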SoundCue /Game/Sounds/Fort_Destructibles/ToxicBarrel/ToxicBarrel_Explosion_Cue
root=QualityLevel_0
nodes [
 {
  id=Attenuation_0
  children [Modulator_2]
  properties {
   attenuationSettings=&/Fort_Explosion_Barrel_Distant_Athena_Att
   attenuationOverrides {falloffDistance=3000, attenuationShapeExtents {x=1000}}
  }
 }
 {
  id=Attenuation_1
  children [Modulator_8]
  properties {
   bOverrideAttenuation=true
   attenuationOverrides {
    bAttenuateWithLPF=true
    bEnableListenerFocus=true
    bEnableLogFrequencyScaling=true
    bEnableSubmixSends=true
    distanceAlgorithm=Custom
    falloffDistance=8000
    focusDistanceScale=0.9
    focusVolumeAttenuation=1.2
    lPFFrequencyAtMax=250
    lPFRadiusMax=8000
    lPFRadiusMin=2000
    nonFocusVolumeAttenuation=0.5
    reverbWetLevelMin=0.2
    attenuationShapeExtents {x=0}
    customAttenuationCurve {editorCurveData {keys [
       <at 0 1>
       {arriveTangent=-1.4319, interpMode=RCIM_Cubic, leaveTangent=-1.4319, tangentMode=RCTM_User, time=0.05, value=1}
       {arriveTangent=-0.8176, interpMode=RCIM_Cubic, leaveTangent=-0.8176, tangentMode=RCTM_User, time=0.3, value=0.3179}
       {arriveTangent=-0.2872, interpMode=RCIM_Cubic, leaveTangent=-0.2872, tangentMode=RCTM_User, time=0.9973, value=0.0145}
      ]}}
    submixSendSettings [{
      bEnableHPFCutoff=false
      bEnableLPFCutoff=false
      disableManualSendClamp=false
      hPFCutoff=20
      lPFCutoff=20000
      maxSendDistance=7500
      maxSendLevel=0.25
      minSendDistance=0
      minSendLevel=0.05
      sendLevel=0.2
      sendLevelControlMethod=Linear
      soundSubmix=&Fortnite_Reverb_Distant
     }]
   }
  }
 }
 {
  id=Attenuation_10
  children [Modulator_6]
  properties {
   bOverrideAttenuation=true
   attenuationOverrides {
    bEnableFocusInterpolation=true
    falloffDistance=400
    focusDistanceScale=0.9
    focusReleaseInterpSpeed=0.1
    nonFocusVolumeAttenuation=0.2
    nonSpatializedRadiusStart=600
    occlusionInterpolationTime=1
    occlusionLowPassFilterFrequency=500
    occlusionTraceChannel=ECC_GameTraceChannel11
    occlusionVolumeAttenuation=0
    reverbWetLevelMin=0.2
    attenuationShapeExtents {x=200}
   }
  }
 }
 {
  id=Attenuation_2
  children [Modulator_6]
  properties {
   bOverrideAttenuation=true
   attenuationOverrides {
    bEnableFocusInterpolation=true
    bEnableListenerFocus=true
    falloffDistance=1600
    focusDistanceScale=0.9
    focusReleaseInterpSpeed=0.1
    nonFocusVolumeAttenuation=0.2
    occlusionInterpolationTime=1
    occlusionLowPassFilterFrequency=500
    occlusionTraceChannel=ECC_GameTraceChannel11
    occlusionVolumeAttenuation=0
    reverbWetLevelMin=0.2
    attenuationShapeExtents {x=500}
   }
  }
 }
 {
  id=Attenuation_4
  children [Modulator_0]
  properties {
   attenuationSettings=&/Fort_Explosion_Barrel_Close_Athena_Att
   attenuationOverrides {falloffDistance=5106.2798, stereoSpread=5084.1587, attenuationShapeExtents {x=2554.7803}}
  }
 }
 {
  id=Attenuation_5
  children [Modulator_5]
  properties {
   attenuationSettings=&/Fort_Explosion_Barrel_Medium_Athena_Att
   attenuationOverrides {falloffDistance=3000, attenuationShapeExtents {x=1000}}
  }
 }
 {
  id=Attenuation_6
  children [Modulator_4]
  properties {
   bOverrideAttenuation=true
   attenuationOverrides {
    bAttenuateWithLPF=true
    bEnableListenerFocus=true
    bEnableLogFrequencyScaling=true
    distanceAlgorithm=Custom
    falloffDistance=8000
    focusDistanceScale=0.9
    focusVolumeAttenuation=1.2
    lPFFrequencyAtMax=500
    lPFRadiusMax=8000
    lPFRadiusMin=2000
    nonFocusVolumeAttenuation=0.5
    reverbWetLevelMin=0.2
    attenuationShapeExtents {x=0}
    customAttenuationCurve {editorCurveData {keys [
       <at 0.25 1 RCIM_Cubic>
       {arriveTangent=-1.316, interpMode=RCIM_Cubic, leaveTangent=-1.316, time=0.5, value=0.35}
       {arriveTangent=-0.5129, interpMode=RCIM_Cubic, leaveTangent=-0.5129, tangentMode=RCTM_User, time=1, value=0.013}
      ]}}
   }
  }
 }
 {
  id=Attenuation_7
  children [Modulator_10]
  properties {
   attenuationSettings=&Fort_Explosion_Barrel_Close_Athena_Att
   attenuationOverrides {falloffDistance=5106.2798, stereoSpread=5084.1587, attenuationShapeExtents {x=2554.7803}}
  }
 }
 {
  id=Attenuation_8
  children [Modulator_11]
  properties {
   attenuationSettings=&Fort_Explosion_Barrel_Medium_Athena_Att
   attenuationOverrides {falloffDistance=3000, attenuationShapeExtents {x=1000}}
  }
 }
 {
  id=Attenuation_9
  children [Modulator_13]
  properties {
   attenuationSettings=&Fort_Explosion_Barrel_Distant_Athena_Att
   attenuationOverrides {falloffDistance=3000, attenuationShapeExtents {x=1000}}
  }
 }
 {id=Mixer_0, children [Attenuation_1, Mixer_3, Modulator_1], properties {inputVolume [1,1,1]}}
 {id=Mixer_1, children [Attenuation_6, Modulator_9], properties {inputVolume [1,0.8]}}
 {id=Mixer_2, children [SoundClass_4,SoundClass_0,SoundClass_5], properties {inputVolume [1,0.75,1]}}
 {id=Mixer_3, children [Attenuation_10, Attenuation_2], properties {inputVolume [1.2,1]}}
 {id=Mixer_4, children [SoundClass_3,SoundClass_2,SoundClass_1], properties {inputVolume [1,0.75,0.5]}}
 {id=Modulator_0, children [Random_7], properties {pitchMax=1.1, pitchMin=0.9}}
 {id=Modulator_1, children [Attenuation_4], properties {pitchMax=1.2, pitchMin=0.9}}
 {id=Modulator_10, children [Random_1], properties {pitchMax=1.1, pitchMin=0.9}}
 {id=Modulator_11, children [Random_9], properties {pitchMax=1.2, pitchMin=1}}
 {id=Modulator_12, children [Attenuation_9], properties {pitchMax=1.1, pitchMin=1}}
 {id=Modulator_13, children [Random_11], properties {volumeMax=0.4, volumeMin=0.35}}
 {id=Modulator_2, children [Modulator_3], properties {pitchMax=1.1, pitchMin=1}}
 {id=Modulator_3, children [Random_5], properties {volumeMax=0.4, volumeMin=0.35}}
 {id=Modulator_4, children [Random_0], properties {pitchMax=1.1, pitchMin=0.9}}
 {id=Modulator_5, children [Random_8], properties {pitchMax=1.2, pitchMin=1}}
 {id=Modulator_6, children [Random_10], properties {pitchMax=1.1, pitchMin=0.9, volumeMax=1.1, volumeMin=1}}
 {id=Modulator_8, children [Random_3], properties {pitchMax=1.1, pitchMin=0.9}}
 {id=Modulator_9, children [Attenuation_7], properties {pitchMax=1.2, pitchMin=0.9}}
 {id=QualityLevel_0, children [Mixer_4,Mixer_4,Mixer_2,Mixer_2]}
 {id=Random_0, children [WavePlayer_6,WavePlayer_8], properties {weights [1,1]}}
 {id=Random_1, children [WavePlayer_5,WavePlayer_4], properties {weights [1,1]}}
 {id=Random_10, children [WavePlayer_40, WavePlayer_41, WavePlayer_42, WavePlayer_43], properties {weights [1,1,1,1]}}
 {id=Random_11, children [WavePlayer_14, WavePlayer_32], properties {weights [1,1]}}
 {id=Random_3, children [WavePlayer_7, WavePlayer_15, WavePlayer_16, WavePlayer_17], properties {weights [1,1,1,1]}}
 {id=Random_5, children [WavePlayer_21, WavePlayer_22], properties {weights [1,1]}}
 {id=Random_7, children [WavePlayer_28, WavePlayer_29, WavePlayer_30, WavePlayer_31], properties {weights [1,1,1,1]}}
 {id=Random_8, children [WavePlayer_9, WavePlayer_10, WavePlayer_11], properties {weights [1,1,1]}}
 {id=Random_9, children [WavePlayer_12, WavePlayer_13], properties {weights [1,1]}}
 {
  id=SoundClass_0
  children [Attenuation_8]
  properties {soundClassOverride=&Fort_WeaponRanged_Fire_Distant}
 }
 {
  id=SoundClass_1
  children [Attenuation_0]
  properties {soundClassOverride=&Fort_WeaponRanged_Fire_Distant}
 }
 {
  id=SoundClass_2
  children [Attenuation_5]
  properties {soundClassOverride=&Fort_WeaponRanged_Fire_Distant}
 }
 {id=SoundClass_3, children [Mixer_0], properties {soundClassOverride=&Fort_WeaponRanged_Fire_3P_PSM}}
 {id=SoundClass_4, children [Mixer_1], properties {soundClassOverride=&Fort_WeaponRanged_Fire_3P_PSM}}
 {
  id=SoundClass_5
  children [Modulator_12]
  properties {soundClassOverride=&Fort_WeaponRanged_Fire_Distant}
 }
 {id=WavePlayer_10, properties {soundWaveAssetPtr=&/Spot_Explosion_Toxic_Barrel_Distant_02}}
 {id=WavePlayer_11, properties {soundWaveAssetPtr=&/Spot_Explosion_Toxic_Barrel_Distant_03}}
 {id=WavePlayer_12, properties {soundWaveAssetPtr=&/Spot_Explosion_Toxic_Barrel_Distant_01}}
 {id=WavePlayer_13, properties {soundWaveAssetPtr=&Spot_Explosion_Toxic_Barrel_Distant_02}}
 {id=WavePlayer_14, properties {soundWaveAssetPtr=&/Spot_Explosion_Toxic_Barrel_Distant_04}}
 {id=WavePlayer_15, properties {soundWaveAssetPtr=&/Spot_Explosion_Toxic_Barrel_Liquid_Goop_Wet_03}}
 {id=WavePlayer_16, properties {soundWaveAssetPtr=&/Spot_Explosion_Toxic_Barrel_Liquid_Goop_Wet_02}}
 {id=WavePlayer_17, properties {soundWaveAssetPtr=&/Spot_Explosion_Toxic_Barrel_Liquid_Goop_Wet_01}}
 {id=WavePlayer_21, properties {soundWaveAssetPtr=&Spot_Explosion_Toxic_Barrel_Distant_04}}
 {id=WavePlayer_22, properties {soundWaveAssetPtr=&/Spot_Explosion_Toxic_Barrel_Distant_05}}
 {id=WavePlayer_28, properties {soundWaveAssetPtr=&/Spot_Explosion_Toxic_Barrel_Explosion_Close_01}}
 {id=WavePlayer_29, properties {soundWaveAssetPtr=&/Spot_Explosion_Toxic_Barrel_Explosion_Close_02}}
 {id=WavePlayer_30, properties {soundWaveAssetPtr=&/Spot_Explosion_Toxic_Barrel_Explosion_Close_03}}
 {id=WavePlayer_31, properties {soundWaveAssetPtr=&/Spot_Explosion_Toxic_Barrel_Explosion_Close_04}}
 {id=WavePlayer_32, properties {soundWaveAssetPtr=&Spot_Explosion_Toxic_Barrel_Distant_05}}
 {id=WavePlayer_4, properties {soundWaveAssetPtr=&Spot_Explosion_Toxic_Barrel_Explosion_Close_02}}
 {id=WavePlayer_40, properties {soundWaveAssetPtr=&/Spot_Explosion_Toxic_Barrel_Liquid_Decal_Swt_01}}
 {id=WavePlayer_41, properties {soundWaveAssetPtr=&/Spot_Explosion_Toxic_Barrel_Liquid_Decal_Swt_02}}
 {id=WavePlayer_42, properties {soundWaveAssetPtr=&/Spot_Explosion_Toxic_Barrel_Liquid_Decal_Swt_03}}
 {id=WavePlayer_43, properties {soundWaveAssetPtr=&/Spot_Explosion_Toxic_Barrel_Liquid_Decal_Swt_04}}
 {id=WavePlayer_5, properties {soundWaveAssetPtr=&Spot_Explosion_Toxic_Barrel_Explosion_Close_01}}
 {id=WavePlayer_6, properties {soundWaveAssetPtr=&/Spot_Explosion_Toxic_Barrel_Liquid_Goop_Wet_04}}
 {id=WavePlayer_7, properties {soundWaveAssetPtr=&Spot_Explosion_Toxic_Barrel_Liquid_Goop_Wet_04}}
 {id=WavePlayer_8, properties {soundWaveAssetPtr=&Spot_Explosion_Toxic_Barrel_Liquid_Goop_Wet_03}}
 {id=WavePlayer_9, properties {soundWaveAssetPtr=&Spot_Explosion_Toxic_Barrel_Distant_01}}
]
properties {
 bOverrideConcurrency=true
 maxDistance=2097152
 volumeMultiplier=0.7
 concurrencyOverrides {maxCount=10, volumeScale=0.8762}
 soundSubmixSends [{minSendLevel=0, soundSubmix=&Fortnite_Reverb_Default}]
}
